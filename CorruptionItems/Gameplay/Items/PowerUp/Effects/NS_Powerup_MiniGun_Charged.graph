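# Niagara system for the charged state of the minigun power-up, with electric ribbons and barrel glow, though all four emitters are disabled.
# Three near-identical ribbon emitters use partitioned BoneFlow modules to draw jittered, magenta-to-violet arcs along separate user-supplied skeletal meshes, one named for overheat.
# A Glow emitter bursts camera-offset sprites on the six front barrel bones and fades them out.
# User.Overheat is declared but unreferenced in the shown modules.
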NiagaraSystem /CorruptionItems/Gameplay/Items/PowerUp/Effects/NS_Powerup_MiniGun_Charged
dependencies {
 usedDataInterfaces [&NiagaraDataInterfaceCurve, &NiagaraDataInterfaceColorCurve]
 usedDynamicInputs []
 usedModules [&SystemState]
}
emitters {
 ElectricityBoneFlow2 {
  bEnabled=false
  emitterSpawnScript [
   /SetVariables(Emitter.RandomSeed:NiagaraInt32=/Niagara/DynamicInputs/Random/RandomInt(Randomness Mode:ENiagaraRandomnessMode=::NewEnumerator0, Random Seed:NiagaraInt32=0, Fixed Random Seed:NiagaraBool=false)) {isSetParameters=true, module=SetVariables_#1}
   /Game/Effects/Niagara/Modules/Partition/FN_PartitionAttributes(Number of Bone Sets=1, Particles Per Partition=8)
  ]
  emitterUpdateScript [
   &EmitterState(Inactive Response=::NewEnumerator1, Life Cycle Mode=::NewEnumerator1, Loop Duration=0.4, UseLoopDelay=false)
   &SpawnBurst_Instantaneous(
    Spawn Count=$Emitter.FN_PartitionAttributes.Spawn Total
    Spawn Time=&RandomRangeFloat(Minimum=0, Maximum=0.1)
    Use Spawn Probability=true
   )
  ]
  particleSpawnScript [
   /Game/Effects/Niagara/Modules/Partition/FN_Particle_partition_v5/FN_Particle_Partition_v5(Enable Ribbon ID=true, Pattern:ENiagaraPartitionWaveForm=::NewEnumerator1, Seed:NiagaraInt32=$Emitter.RandomSeed)
   &InitializeParticle(Lifetime=0.3, UsePositionOffset=false, Write Material Random=true)
   &InitializeRibbon(
    Color Maximum:LinearColor=<rgba 1 0.0603 0.9063 1>
    Color Minimum:LinearColor=<rgba 0.0221 0 0.3 1>
    Color Mode:ENiagara_ColorInitializationMode=::NewEnumerator2
    Lifetime Max=0.75
    Lifetime Min=0.2
    Lifetime Mode:ENiagara_LifetimeMode=::NewEnumerator1
    Ribbon Width=&RandomRangeFloat(Minimum=6, Maximum=13)
    UsePositionOffset=false
    Write Material Random=true
   )
   /Game/Effects/Niagara/Systems/Prototype/Tests/MusicDrivenCCPM/FN_Initialize_BoneFlow_Attributes(
    Angle Count Min=-1
    Angle Offset Min=-1
    Max T Rate=1
    MaxTScale=1
    Number of Bone Sets=1
    Radius Multiplier:NiagaraFloat=&RandomRangeFloat(Minimum=-10, Maximum=10)
    RadiusMax=1
    RadiusMin=0.2
    Rate=1
   )
   &DynamicMaterialParameters(
    Index 0 Param 2=&FloatFromCurve(FloatCurve=DataInterface() {propertyValues {Curve {keys [<at 0 0>,<at 1 -0.8>]}}})
    Index 0 Param 3=&RandomRangeFloat(Minimum=0.3, Maximum=0.7)
    Module.Param0WriteEnabled=false
    Module.Param1WriteEnabled=true
    Module.Param2WriteEnabled=true
    Module.Param3WriteEnabled=false
   )
  ]
  particleUpdateScript [
   &ParticleState()
   &FN_BoneFlow(
    Angle:NiagaraFloat=$Particles.FN_Initialize_BoneFlow_Attributes.TScale
    BoneID:NiagaraInt32=$Particles.Partition.ID
    Enable Position Offsets=true
    Height:NiagaraFloat=$Particles.Initial.Partition.NormalizedID
    Offset End:Vector3f=&RandomRangeVector(Minimum=<0 0 0>, Maximum=<0 0 0>)
    Offset Mid:Vector3f=&RandomRangeVector(Minimum=<-2 -2 -2>, Maximum=<2 2 2>)
    Offset Start:Vector3f=&RandomRangeVector(Minimum=<0 0 0>, Maximum=<0 0 0>)
    Radius:NiagaraFloat=$Particles.Initial.FN_Initialize_BoneFlow_Attributes.Radius
    Set Material Dynamic=false
    SkeletalMesh:NiagaraDataInterfaceSkeletalMesh=$User.SkeletalMesh2
   )
   &ScaleColor(
    Scale Mode=::NewEnumerator1
    Scale RGBA=&Vector4FromCurve(Vector4Curve=DataInterface() {propertyValues {
       AlphaCurve {keys [<at 0 0>,<at 0.2262 1>,<at 0.4885 1>,<at 1 1>]}
       BlueCurve {keys [<at 0.3099 15>,<at 0.3685 5>,<at 0.9991 1>]}
       GreenCurve {keys [<at 0.3099 15>,<at 0.3685 5>,<at 0.9991 1>]}
       RedCurve {keys [<at 0.3099 15>,<at 0.3685 5>,<at 0.9991 1>]}
      }})
    ScaleRGBA=true
   ) {module=ScaleColor001}
   &ScaleRibbonWidth(Ribbon Width Scale=2)
  ]
  rendererClasses [&NiagaraRibbonRendererProperties]
  renderers [{rendererClass=&NiagaraRibbonRendererProperties}]
 }
 ElectricityBoneFlow3 {
  bEnabled=false
  emitterSpawnScript [
   /SetVariables(Emitter.RandomSeed:NiagaraInt32=&RandomInt(Randomness Mode:ENiagaraRandomnessMode=::NewEnumerator0, Random Seed:NiagaraInt32=0, Fixed Random Seed:NiagaraBool=false)) {isSetParameters=true, module=SetVariables_#1}
   &FN_PartitionAttributes(Number of Bone Sets=1, Particles Per Partition=8)
  ]
  emitterUpdateScript [
   &EmitterState(Inactive Response=::NewEnumerator1, Life Cycle Mode=::NewEnumerator1, Loop Duration=0.4, UseLoopDelay=false)
   &SpawnBurst_Instantaneous(
    Spawn Count=$Emitter.FN_PartitionAttributes.Spawn Total
    Spawn Time=&RandomRangeFloat(Minimum=0, Maximum=0.1)
    Use Spawn Probability=true
   )
  ]
  particleSpawnScript [
   &FN_Particle_Partition_v5(Enable Ribbon ID=true, Pattern:ENiagaraPartitionWaveForm=::NewEnumerator1, Seed:NiagaraInt32=$Emitter.RandomSeed)
   &InitializeParticle(Lifetime=0.3, UsePositionOffset=false, Write Material Random=true)
   &InitializeRibbon(
    Color Maximum:LinearColor=<rgba 1 0.0603 0.9063 1>
    Color Minimum:LinearColor=<rgba 0.0221 0 0.3 1>
    Color Mode:ENiagara_ColorInitializationMode=::NewEnumerator2
    Lifetime Max=0.75
    Lifetime Min=0.2
    Lifetime Mode:ENiagara_LifetimeMode=::NewEnumerator1
    Ribbon Width=&RandomRangeFloat(Minimum=6, Maximum=13)
    UsePositionOffset=false
    Write Material Random=true
   )
   &FN_Initialize_BoneFlow_Attributes(
    Angle Count Min=-1
    Angle Offset Min=-1
    Max T Rate=1
    MaxTScale=1
    Number of Bone Sets=1
    Radius Multiplier:NiagaraFloat=&RandomRangeFloat(Minimum=-10, Maximum=10)
    RadiusMax=1
    RadiusMin=0.2
    Rate=1
   )
   &DynamicMaterialParameters(
    Index 0 Param 2=&FloatFromCurve(FloatCurve=DataInterface() {propertyValues {Curve {keys [<at 0 0>,<at 1 -0.8>]}}})
    Index 0 Param 3=&RandomRangeFloat(Minimum=0.3, Maximum=0.7)
    Module.Param0WriteEnabled=false
    Module.Param1WriteEnabled=true
    Module.Param2WriteEnabled=true
    Module.Param3WriteEnabled=false
   )
  ]
  particleUpdateScript [
   &ParticleState()
   &FN_BoneFlow(
    Angle:NiagaraFloat=$Particles.FN_Initialize_BoneFlow_Attributes.TScale
    BoneID:NiagaraInt32=$Particles.Partition.ID
    Enable Position Offsets=true
    Height:NiagaraFloat=$Particles.Initial.Partition.NormalizedID
    Offset End:Vector3f=&RandomRangeVector(Minimum=<0 0 0>, Maximum=<0 0 0>)
    Offset Mid:Vector3f=&RandomRangeVector(Minimum=<-2 -2 -2>, Maximum=<2 2 2>)
    Offset Start:Vector3f=&RandomRangeVector(Minimum=<0 0 0>, Maximum=<0 0 0>)
    Radius:NiagaraFloat=$Particles.Initial.FN_Initialize_BoneFlow_Attributes.Radius
    Set Material Dynamic=false
    SkeletalMesh:NiagaraDataInterfaceSkeletalMesh=$User.SkeletalMesh3
   )
   &ScaleColor(
    Scale Mode=::NewEnumerator1
    Scale RGBA=&Vector4FromCurve(Vector4Curve=DataInterface() {propertyValues {
       AlphaCurve {keys [<at 0 0>,<at 0.2262 1>,<at 0.4885 1>,<at 1 1>]}
       BlueCurve {keys [<at 0.3099 15>,<at 0.3685 5>,<at 0.9991 1>]}
       GreenCurve {keys [<at 0.3099 15>,<at 0.3685 5>,<at 0.9991 1>]}
       RedCurve {keys [<at 0.3099 15>,<at 0.3685 5>,<at 0.9991 1>]}
      }})
    ScaleRGBA=true
   ) {module=ScaleColor001}
   &ScaleRibbonWidth(Ribbon Width Scale=2)
  ]
  rendererClasses [&NiagaraRibbonRendererProperties]
  renderers [{rendererClass=&NiagaraRibbonRendererProperties}]
 }
 ElectricityBoneFlowOverheat {
  bEnabled=false
  emitterSpawnScript [
   /SetVariables(Emitter.RandomSeed:NiagaraInt32=&RandomInt(Randomness Mode:ENiagaraRandomnessMode=::NewEnumerator0, Random Seed:NiagaraInt32=0, Fixed Random Seed:NiagaraBool=false)) {isSetParameters=true, module=SetVariables_#1}
   &FN_PartitionAttributes(Number of Bone Sets=1, Particles Per Partition=8)
  ]
  emitterUpdateScript [
   &EmitterState(Inactive Response=::NewEnumerator1, Life Cycle Mode=::NewEnumerator1, Loop Duration=0.4, UseLoopDelay=false)
   &SpawnBurst_Instantaneous(
    Spawn Count=$Emitter.FN_PartitionAttributes.Spawn Total
    Spawn Time=&RandomRangeFloat(Minimum=0, Maximum=0.1)
    Use Spawn Probability=true
   )
  ]
  particleSpawnScript [
   &FN_Particle_Partition_v5(Enable Ribbon ID=true, Pattern:ENiagaraPartitionWaveForm=::NewEnumerator1, Seed:NiagaraInt32=$Emitter.RandomSeed)
   &InitializeParticle(Lifetime=0.3, UsePositionOffset=false, Write Material Random=true)
   &InitializeRibbon(
    Color Maximum:LinearColor=<rgba 1 0.0603 0.9063 1>
    Color Minimum:LinearColor=<rgba 0.0221 0 0.3 1>
    Color Mode:ENiagara_ColorInitializationMode=::NewEnumerator2
    Lifetime Max=0.75
    Lifetime Min=0.2
    Lifetime Mode:ENiagara_LifetimeMode=::NewEnumerator1
    Ribbon Width=&RandomRangeFloat(Minimum=6, Maximum=13)
    UsePositionOffset=false
    Write Material Random=true
   )
   &FN_Initialize_BoneFlow_Attributes(
    Angle Count Min=-1
    Angle Offset Min=-1
    Max T Rate=1
    MaxTScale=1
    Number of Bone Sets=1
    Radius Multiplier:NiagaraFloat=&RandomRangeFloat(Minimum=-10, Maximum=10)
    RadiusMax=1
    RadiusMin=0.2
    Rate=1
   )
   &DynamicMaterialParameters(
    Index 0 Param 2=&FloatFromCurve(FloatCurve=DataInterface() {propertyValues {Curve {keys [<at 0 0>,<at 1 -0.8>]}}})
    Index 0 Param 3=&RandomRangeFloat(Minimum=0.3, Maximum=0.7)
    Module.Param0WriteEnabled=false
    Module.Param1WriteEnabled=true
    Module.Param2WriteEnabled=true
    Module.Param3WriteEnabled=false
   )
  ]
  particleUpdateScript [
   &ParticleState()
   &FN_BoneFlow(
    Angle:NiagaraFloat=$Particles.FN_Initialize_BoneFlow_Attributes.TScale
    BoneID:NiagaraInt32=$Particles.Partition.ID
    Enable Position Offsets=true
    Height:NiagaraFloat=$Particles.Initial.Partition.NormalizedID
    Offset End:Vector3f=&RandomRangeVector(Minimum=<0 0 0>, Maximum=<0 0 0>)
    Offset Mid:Vector3f=&RandomRangeVector(Minimum=<-2 -2 -2>, Maximum=<2 2 2>)
    Offset Start:Vector3f=&RandomRangeVector(Minimum=<0 0 0>, Maximum=<0 0 0>)
    Radius:NiagaraFloat=$Particles.Initial.FN_Initialize_BoneFlow_Attributes.Radius
    Set Material Dynamic=false
    SkeletalMesh:NiagaraDataInterfaceSkeletalMesh=$User.SkeletalMesh_Overheat
   )
   &ScaleColor(
    Scale Mode=::NewEnumerator1
    Scale RGBA=&Vector4FromCurve(Vector4Curve=DataInterface() {propertyValues {
       AlphaCurve {keys [<at 0 0>,<at 0.2262 1>,<at 0.4885 1>,<at 1 1>]}
       BlueCurve {keys [<at 0.3099 15>,<at 0.3685 5>,<at 0.9991 1>]}
       GreenCurve {keys [<at 0.3099 15>,<at 0.3685 5>,<at 0.9991 1>]}
       RedCurve {keys [<at 0.3099 15>,<at 0.3685 5>,<at 0.9991 1>]}
      }})
    ScaleRGBA=true
   ) {module=ScaleColor001}
   &ScaleRibbonWidth(Ribbon Width Scale=2)
  ]
  rendererClasses [&NiagaraRibbonRendererProperties]
  renderers [{rendererClass=&NiagaraRibbonRendererProperties}]
 }
 Glow {
  bEnabled=false
  emitterUpdateScript [
   &EmitterState(Life Cycle Mode=::NewEnumerator1, Loop Duration=0.6, UseLoopDelay=false)
   &SpawnBurst_Instantaneous(Spawn Count=6, Use Spawn Probability=true, UseLoopCountLimit=false)
  ]
  particleSpawnScript [
   &FN_SetSpawnGroup(Number of Spawn Groups=6)
   &InitializeParticle(
    Color=<rgba 1 0.0603 0.9063 1>
    Lifetime=0.75
    Sprite Size Mode=::NewEnumerator3
    Uniform Sprite Size=20
    UsePositionOffset=false
    Write Material Random=true
   )
   &CameraOffset(Camera Offset Amount=10)
  ]
  particleUpdateScript [
   &ParticleState()
   &SkeletalMeshLocation(
    Bone / Socket Index=$Particles.SpawnGroup
    Bone Sampling Mode=::NewEnumerator4
    MaskSpawnGroup=false
    Skeletal Mesh=DataInterface() {propertyValues {
      DefaultMesh=null
      ExcludeBoneName=root
      OptionalComponentClass=null
      PreviewMesh=/CorruptionItems/Gameplay/Items/PowerUp/LMG/Meshes/Power_Up_Minigun_Gun
      SkinningMode=SkinOnTheFly
      SoftSourceActor=null
      SourceMode=Default
      UvSetIndex=0
      WholeMeshLOD=-1
      bExcludeBone=false
      bReadDeformedGeometry=true
      bRequireCurrentFrameData=true
      ComponentTags []
      FilteredBones [BarrelFront1,BarrelFront2,BarrelFront3,BarrelFront4,BarrelFront5,BarrelFront6]
      FilteredSockets []
      MeshUserParameter {parameter {name=None, typeDefHandle {registeredTypeIndex=1}, varData []}}
      SamplingRegions []
     }}
   )
   &ScaleColor(
    Scale Alpha=&FloatFromCurve(FloatCurve=DataInterface() {propertyValues {Curve {keys [<at 0.6 1 RCIM_Cubic>, {arriveTangent=-5.6404, interpMode=RCIM_Cubic, tangentMode=RCTM_Break, time=1, value=0}]}}})
    ScaleA=true
    ScaleRGB=false
   )
  ]
  rendererClasses [&NiagaraSpriteRendererProperties]
  renderers [{rendererClass=&NiagaraSpriteRendererProperties}]
 }
}
summary {
 systemName=NS_Powerup_MiniGun_Charged
 emitters [
  {bEnabled=false, emitterName=ElectricityBoneFlowOverheat, rendererClasses [&NiagaraRibbonRendererProperties]}
  {bEnabled=false, emitterName=ElectricityBoneFlow2, rendererClasses [&NiagaraRibbonRendererProperties]}
  {bEnabled=false, emitterName=ElectricityBoneFlow3, rendererClasses [&NiagaraRibbonRendererProperties]}
  {bEnabled=false, emitterName=Glow, rendererClasses [&NiagaraSpriteRendererProperties]}
 ]
 userVariables [
  User.Overheat {type=NiagaraFloat, defaultValue=1, description=""}
  User.SkeletalMesh2 {
   type=NiagaraDataInterfaceSkeletalMesh
   defaultValue {
    struct=NiagaraExt_VariableValue_DataInterface
    value {dataInterface=&self:NiagaraDataInterfaceSkeletalMesh_0}
   }
  }
  User.SkeletalMesh3 {
   type=NiagaraDataInterfaceSkeletalMesh
   defaultValue {
    struct=NiagaraExt_VariableValue_DataInterface
    value {dataInterface=&self:NiagaraDataInterfaceSkeletalMesh_1}
   }
  }
  User.SkeletalMesh_Overheat {
   type=NiagaraDataInterfaceSkeletalMesh
   defaultValue {
    struct=NiagaraExt_VariableValue_DataInterface
    value {dataInterface=&self:NiagaraDataInterfaceSkeletalMesh_2}
   }
  }
 ]
}
