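# Niagara idle effect for the minigun power-up: electric sprites, ribbon arcs, glow and small mesh shards on the gun, driven by User.Overheat.
# Sprites, glow and shards spawn on filtered body-piece bones, while two partitioned-ribbon emitters use FN_BoneFlow along the main and rear skeletal-mesh inputs.
# Overheat thresholds gate most spawning; its value also scales shard spawn probability, glow alpha and ribbon width.
# All emitters tint toward a shared system overheat color.
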NiagaraSystem /CorruptionItems/Gameplay/Items/PowerUp/Effects/NS_Powerup_MiniGun_Idle
dependencies {
 usedDataInterfaces [&NiagaraDataInterfaceCurve, &NiagaraDataInterfaceColorCurve]
 usedDynamicInputs []
 usedModules [
  &SystemState
  &self:SystemSpawnScript.SystemScriptSource.SystemScriptGraph.NiagaraNodeAssignment_2.SetVariables_#1
  &self:SystemSpawnScript.SystemScriptSource.SystemScriptGraph.NiagaraNodeAssignment_0.SetVariables_#2
  &self:SystemSpawnScript.SystemScriptSource.SystemScriptGraph.NiagaraNodeAssignment_1.SetVariables_#3
 ]
}
emitters {
 ElectricSprites {
  emitterUpdateScript [
   &EmitterState()
   &SpawnRate(
    Spawn Probability=&MakeCustomFloatFromBool(Bool=&SetBoolByFloatComparison(A=$User.Overheat, B=0.35), True Float=1, False Float=0)
    SpawnRate=25
    Use Spawn Probability=true
   )
  ]
  particleSpawnScript [
   &FN_SetSpawnGroup()
   /SetVariables(Particles.ColorStart:LinearColor=<rgba 12 0.7236 10.876 1>) {isSetParameters=true}
   &InitializeParticle(
    Lifetime Max=0.3
    Lifetime Min=0.15
    Lifetime Mode=::NewEnumerator1
    Sprite Rotation Mode=::NewEnumerator2
    Sprite Size=&Multiply_Vector2D_ByFloat(
     Float=&RandomRangeFloat(Minimum=0.35, Maximum=0.6)
     Vector2D=&RandomRangeVector2D(Minimum=<35 35>, Maximum=<55 55>)
    )
    Sprite Size Mode=::NewEnumerator4
    UsePositionOffset=false
    Write Material Random=true
   )
   /Niagara/Modules/Update/SubUV/SubUVAnimation(SubUV Animation Mode=::NewEnumerator1)
   &DynamicMaterialParameters(
    Index 0 Param 1=$Particles.MaterialRandom
    Index 0 Param 2=&Lerp_Float(A=2, B=4, Alpha=$Particles.MaterialRandom)
    Module.Param0WriteEnabled=true
    Module.Param1WriteEnabled=true
    Module.Param2WriteEnabled=false
    Module.Param3WriteEnabled=false
   )
   &CameraOffset(Camera Offset Amount=5)
   &SkeletalMeshLocation(
    Bone / Socket Index=$Particles.SpawnGroup
    Bone Sampling Mode=::NewEnumerator4
    Interpolate Mesh Sampling=false
    MaskSpawnGroup=false
    Skeletal Mesh=DataInterface() {propertyValues {
      DefaultMesh=null
      ExcludeBoneName=root
      OptionalComponentClass=null
      PreviewMesh=/CorruptionItems/Gameplay/Items/PowerUp/LMG/Meshes/Power_Up_Minigun_Gun
      SkinningMode=SkinOnTheFly
      SoftSourceActor=null
      SourceMode=Default
      UvSetIndex=0
      WholeMeshLOD=-1
      bExcludeBone=false
      bReadDeformedGeometry=true
      bRequireCurrentFrameData=true
      ComponentTags []
      FilteredBones [BodyPieceTop_l, BodyPieceTop_r, BodyPieceBottom_l, BodyPieceBottom_r]
      FilteredSockets []
      MeshUserParameter {parameter {name=None, typeDefHandle {registeredTypeIndex=1}, varData []}}
      SamplingRegions []
     }}
   )
  ]
  particleUpdateScript [
   &SolveForcesAndVelocity(Clamp Velocity=false, Limit Acceleration=false)
   &ParticleState()
   &DynamicMaterialParameters(
    Index 0 Param 3=&FloatFromCurve(FloatCurve=DataInterface() {propertyValues {Curve {keys [<at 0.2 0 RCIM_Cubic>, {interpMode=RCIM_Cubic, leaveTangent=1.5819, time=1, value=1}]}}})
    Module.Param0WriteEnabled=false
    Module.Param1WriteEnabled=false
    Module.Param2WriteEnabled=true
    Module.Param3WriteEnabled=false
   ) {module=DynamicMaterialParameters001}
   &ScaleSpriteSize(Scale Factor=&Vector2DFromFloat(Value=&FloatFromCurve(FloatCurve=DataInterface() {propertyValues {Curve {keys [{arriveTangent=0.3147, interpMode=RCIM_Cubic, leaveTangent=0.3147, tangentMode=RCTM_User, time=0, value=0.8}, <at 1 1>]}}})))
   &ScaleColor(
    Scale RGB=&MakeVectorFromLinearColorRGB(Color=&LerpLinearColors(StartColor=$Particles.ColorStart, EndColor=$System.ColorOverheat, LerpFactor=$System.ColorOverheatCurve))
    ScaleA=true
    ScaleRGB=true
   )
  ]
  rendererClasses [&NiagaraSpriteRendererProperties]
  renderers [{rendererClass=&NiagaraSpriteRendererProperties}]
 }
 ElectricityBoneFlow {
  emitterSpawnScript [
   /SetVariables(Emitter.RandomSeed:NiagaraInt32=/Niagara/DynamicInputs/Random/RandomInt(Randomness Mode:ENiagaraRandomnessMode=::NewEnumerator0, Random Seed:NiagaraInt32=0, Fixed Random Seed:NiagaraBool=false)) {isSetParameters=true, module=SetVariables_#4}
   /Game/Effects/Niagara/Modules/Partition/FN_PartitionAttributes(Number of Bone Sets=1, Particles Per Partition=6)
  ]
  emitterUpdateScript [
   &EmitterState(Life Cycle Mode=::NewEnumerator1, Loop Duration=0.5, UseLoopDelay=false)
   &SpawnBurst_Instantaneous(
    Spawn Count=$Emitter.FN_PartitionAttributes.Spawn Total
    Spawn Probability=&MakeCustomFloatFromBool(Bool=&SetBoolByFloatComparison(A=$User.Overheat, B=0.2), True Float=1, False Float=0)
    Spawn Time=&RandomRangeFloat(Minimum=0, Maximum=0.1)
    Use Spawn Probability=true
   )
  ]
  particleSpawnScript [
   /Game/Effects/Niagara/Modules/Partition/FN_Particle_partition_v5/FN_Particle_Partition_v5(Enable Ribbon ID=true, Pattern:ENiagaraPartitionWaveForm=::NewEnumerator1, Seed:NiagaraInt32=$Emitter.RandomSeed)
   /SetVariables(Particles.ColorStart:LinearColor=&MultiplyLinearColor(
     Linear Color=&RandomRangeLinearColor(Minimum=<rgba 0.0221 0 0.3 1>, Maximum=<rgba 1 0.0603 0.9063 1>)
     Scale Factor=&MakeFloat4FromVectorAndFloat("Vector (RGB)"=&VectorFromFloat(Value=1), "Float (Alpha)"=1)
    )) {isSetParameters=true}
   &InitializeParticle(Lifetime=0.3, UsePositionOffset=false, Write Material Random=true)
   &InitializeRibbon(
    Lifetime Max=0.75
    Lifetime Min=0.2
    Lifetime Mode:ENiagara_LifetimeMode=::NewEnumerator1
    Ribbon Width=&RandomRangeFloat(Minimum=6, Maximum=13)
    UsePositionOffset=false
    Write Material Random=true
   )
   /Game/Effects/Niagara/Systems/Prototype/Tests/MusicDrivenCCPM/FN_Initialize_BoneFlow_Attributes(
    Number of Bone Sets=1
    Radius Multiplier:NiagaraFloat=&RandomRangeFloat(Minimum=-5, Maximum=5)
    RadiusMax=1
    RadiusMin=0.2
    Rate=1
   )
   &DynamicMaterialParameters(
    Index 0 Param 1=1
    Index 0 Param 2=&FloatFromCurve(FloatCurve=DataInterface() {propertyValues {Curve {keys [<at 0 0>,<at 1 -1>]}}})
    Index 0 Param 3=&RandomRangeFloat(Minimum=0.2, Maximum=0.3)
    Module.Param0WriteEnabled=true
    Module.Param1WriteEnabled=true
    Module.Param2WriteEnabled=true
    Module.Param3WriteEnabled=false
   )
  ]
  particleUpdateScript [
   &ParticleState()
   &FN_BoneFlow(
    Angle:NiagaraFloat=$Particles.FN_Initialize_BoneFlow_Attributes.TScale
    BoneID:NiagaraInt32=$Particles.Partition.ID
    Height:NiagaraFloat=$Particles.Initial.Partition.NormalizedID
    Radius:NiagaraFloat=$Particles.Initial.FN_Initialize_BoneFlow_Attributes.Radius
    Set Material Dynamic=false
    SkeletalMesh:NiagaraDataInterfaceSkeletalMesh=$User.SkeletalMesh
   )
   &ScaleColor(
    Scale Mode=::NewEnumerator1
    Scale RGBA=&Vector4FromCurve(Vector4Curve=DataInterface() {propertyValues {
       AlphaCurve {keys [<at 0 0>,<at 0.2262 1>,<at 0.4885 1>,<at 1 1>]}
       BlueCurve {keys [<at 0.3099 20>,<at 0.3685 10>,<at 0.9991 2>]}
       GreenCurve {keys [<at 0.3099 20>,<at 0.3685 10>,<at 0.9991 2>]}
       RedCurve {keys [<at 0.3099 20>,<at 0.3685 10>,<at 0.9991 2>]}
      }})
    ScaleRGBA=true
   )
   &ScaleRibbonWidth(Ribbon Width Scale=&Lerp_Float(A=1.75, B=2.5, Alpha=$User.Overheat))
   &ScaleColor(
    Scale Alpha=&FloatFromCurve(
     FloatCurve=DataInterface() {propertyValues {Curve {keys [<at 0 0>,<at 0.1 1>,<at 0.25 1>,<at 0.3 0>,<at 0.7 0>,<at 0.75 1>,<at 0.9 1>,<at 1 0>]}}}
     CurveIndex=$Particles.Partition.NormalizedID
    )
    Scale RGB=&MakeVectorFromLinearColorRGB(Color=&LerpLinearColors(StartColor=$Particles.ColorStart, EndColor=$System.ColorOverheat, LerpFactor=$System.ColorOverheatCurve))
    ScaleA=true
    ScaleRGB=true
   ) {module=ScaleColor001}
  ]
  rendererClasses [&NiagaraRibbonRendererProperties]
  renderers [{rendererClass=&NiagaraRibbonRendererProperties}]
 }
 ElectricityBoneRear {
  emitterSpawnScript [
   /SetVariables(Emitter.RandomSeed:NiagaraInt32=&RandomInt(Randomness Mode:ENiagaraRandomnessMode=::NewEnumerator0, Random Seed:NiagaraInt32=0, Fixed Random Seed:NiagaraBool=false)) {isSetParameters=true, module=SetVariables_#4}
   &FN_PartitionAttributes(Number of Bone Sets=1, Particles Per Partition=6)
  ]
  emitterUpdateScript [
   &EmitterState(Life Cycle Mode=::NewEnumerator1, Loop Duration=0.5, UseLoopDelay=false)
   &SpawnBurst_Instantaneous(
    Spawn Count=$Emitter.FN_PartitionAttributes.Spawn Total
    Spawn Probability=&MakeCustomFloatFromBool(Bool=&SetBoolByFloatComparison(A=$User.Overheat, B=0.2), True Float=1, False Float=0)
    Spawn Time=&RandomRangeFloat(Minimum=0, Maximum=0.1)
    Use Spawn Probability=true
   )
  ]
  particleSpawnScript [
   &FN_Particle_Partition_v5(Enable Ribbon ID=true, Pattern:ENiagaraPartitionWaveForm=::NewEnumerator1, Seed:NiagaraInt32=$Emitter.RandomSeed)
   /SetVariables(Particles.ColorStart:LinearColor=&MultiplyLinearColor(
     Linear Color=&RandomRangeLinearColor(Minimum=<rgba 0.0221 0 0.3 1>, Maximum=<rgba 1 0.0603 0.9063 1>)
     Scale Factor=&MakeFloat4FromVectorAndFloat("Vector (RGB)"=&VectorFromFloat(Value=1), "Float (Alpha)"=1)
    )) {isSetParameters=true}
   &InitializeParticle(Lifetime=0.3, UsePositionOffset=false, Write Material Random=true)
   &InitializeRibbon(
    Lifetime Max=0.75
    Lifetime Min=0.2
    Lifetime Mode:ENiagara_LifetimeMode=::NewEnumerator1
    Ribbon Width=&RandomRangeFloat(Minimum=6, Maximum=13)
    UsePositionOffset=false
    Write Material Random=true
   )
   &FN_Initialize_BoneFlow_Attributes(
    Number of Bone Sets=1
    Radius Multiplier:NiagaraFloat=&RandomRangeFloat(Minimum=-5, Maximum=5)
    RadiusMax=1
    RadiusMin=0.2
    Rate=1
   )
   &DynamicMaterialParameters(
    Index 0 Param 2=&FloatFromCurve(FloatCurve=DataInterface() {propertyValues {Curve {keys [<at 0 0>,<at 1 -1>]}}})
    Index 0 Param 3=&RandomRangeFloat(Minimum=0.2, Maximum=0.3)
    Module.Param0WriteEnabled=false
    Module.Param1WriteEnabled=true
    Module.Param2WriteEnabled=true
    Module.Param3WriteEnabled=false
   )
  ]
  particleUpdateScript [
   &ParticleState()
   &FN_BoneFlow(
    Angle:NiagaraFloat=$Particles.FN_Initialize_BoneFlow_Attributes.TScale
    BoneID:NiagaraInt32=$Particles.Partition.ID
    Enable Position Offsets=true
    Height:NiagaraFloat=$Particles.Initial.Partition.NormalizedID
    Radius:NiagaraFloat=$Particles.Initial.FN_Initialize_BoneFlow_Attributes.Radius
    Set Material Dynamic=false
    SkeletalMesh:NiagaraDataInterfaceSkeletalMesh=$User.SkeletalMesh_Rear
   )
   &ScaleColor(
    Scale Mode=::NewEnumerator1
    Scale RGBA=&Vector4FromCurve(Vector4Curve=DataInterface() {propertyValues {
       AlphaCurve {keys [<at 0 0>,<at 0.2262 1>,<at 0.4885 1>,<at 1 1>]}
       BlueCurve {keys [<at 0.3099 20>,<at 0.3685 10>,<at 0.9991 2>]}
       GreenCurve {keys [<at 0.3099 20>,<at 0.3685 10>,<at 0.9991 2>]}
       RedCurve {keys [<at 0.3099 20>,<at 0.3685 10>,<at 0.9991 2>]}
      }})
    ScaleRGBA=true
   )
   &ScaleRibbonWidth(Ribbon Width Scale=&Lerp_Float(A=1.75, B=2.5, Alpha=$User.Overheat))
   &ScaleColor(
    Scale Alpha=&FloatFromCurve(
     FloatCurve=DataInterface() {propertyValues {Curve {keys [<at 0 0>,<at 0.1 1>,<at 0.2 1>,<at 0.25 0>,<at 0.75 0>,<at 0.8 1>,<at 0.9 1>,<at 1 0>]}}}
     CurveIndex=$Particles.Partition.NormalizedID
    )
    Scale RGB=&MakeVectorFromLinearColorRGB(Color=&LerpLinearColors(StartColor=$Particles.ColorStart, EndColor=$System.ColorOverheat, LerpFactor=$System.ColorOverheatCurve))
    ScaleA=true
    ScaleRGB=true
   ) {module=ScaleColor001}
  ]
  rendererClasses [&NiagaraRibbonRendererProperties]
  renderers [{rendererClass=&NiagaraRibbonRendererProperties}]
 }
 Glow {
  emitterUpdateScript [
   &EmitterState()
   &SpawnRate(
    Spawn Probability=&MakeCustomFloatFromBool(Bool=&SetBoolByFloatComparison(A=$User.Overheat, B=0.25), True Float=1, False Float=0)
    SpawnRate=2
    Use Spawn Probability=true
   )
  ]
  particleSpawnScript [
   &FN_SetSpawnGroup()
   /SetVariables(Particles.ColorStart:LinearColor=<rgba 1 0.0603 0.9063 1>) {isSetParameters=true}
   &InitializeParticle(
    Lifetime Max=1
    Lifetime Mode=::NewEnumerator1
    Sprite Size Mode=::NewEnumerator3
    Uniform Sprite Size=25
    UsePositionOffset=false
    Write Material Random=true
   )
  ]
  particleUpdateScript [
   &ParticleState()
   &SkeletalMeshLocation(
    Bone / Socket Index=$Particles.SpawnGroup
    Bone Sampling Mode=::NewEnumerator4
    MaskSpawnGroup=false
    Skeletal Mesh=DataInterface() {propertyValues {
      DefaultMesh=null
      ExcludeBoneName=root
      OptionalComponentClass=null
      PreviewMesh=&Power_Up_Minigun_Gun
      SkinningMode=SkinOnTheFly
      SoftSourceActor=null
      SourceMode=Default
      UvSetIndex=0
      WholeMeshLOD=-1
      bExcludeBone=false
      bReadDeformedGeometry=true
      bRequireCurrentFrameData=true
      ComponentTags []
      FilteredBones [BodyPieceTop_l, BodyPieceTop_r, BodyPieceBottom_l, BodyPieceBottom_r]
      FilteredSockets []
      MeshUserParameter {parameter {name=None, typeDefHandle {registeredTypeIndex=1}, varData []}}
      SamplingRegions []
     }}
   )
   &ScaleColor(
    Scale Alpha=&FloatFromCurve(
     FloatCurve=DataInterface() {propertyValues {Curve {keys [<at 0.1 0 RCIM_Cubic>, {interpMode=RCIM_Cubic, tangentMode=RCTM_Break, time=1, value=1}]}}}
     CurveIndex=$User.Overheat
    )
    Scale RGB=&MakeVectorFromLinearColorRGB(Color=&LerpLinearColors(StartColor=$Particles.ColorStart, EndColor=$System.ColorOverheat, LerpFactor=$System.ColorOverheatCurve))
    ScaleA=true
    ScaleRGB=true
   )
  ]
  rendererClasses [&NiagaraSpriteRendererProperties]
  renderers [{rendererClass=&NiagaraSpriteRendererProperties}]
 }
 ShapesSmall {
  emitterUpdateScript [
   &EmitterState()
   &SpawnRate(Spawn Probability=&Lerp_Float(A=0, B=1, Alpha=$User.Overheat), SpawnRate=25, Use Spawn Probability=true)
  ]
  particleSpawnScript [
   &FN_SetSpawnGroup()
   /SetVariables(Particles.ColorStart:LinearColor=<rgba 2 0.1206 1.8127 1>) {isSetParameters=true}
   &InitializeParticle(
    Lifetime Max=0.25
    Lifetime Min=0.1
    Lifetime Mode=::NewEnumerator1
    Mesh Scale Mode=::NewEnumerator1
    Mesh Uniform Scale Max=3
    UsePositionOffset=false
    Write Material Random=true
   )
   &SkeletalMeshLocation(
    Bone / Socket Index=$Particles.SpawnGroup
    Bone Sampling Mode=::NewEnumerator4
    Interpolate Mesh Sampling=false
    MaskSpawnGroup=false
    Skeletal Mesh=DataInterface() {propertyValues {
      DefaultMesh=null
      ExcludeBoneName=root
      OptionalComponentClass=null
      PreviewMesh=&Power_Up_Minigun_Gun
      SkinningMode=SkinOnTheFly
      SoftSourceActor=null
      SourceMode=Default
      UvSetIndex=0
      WholeMeshLOD=-1
      bExcludeBone=false
      bReadDeformedGeometry=true
      bRequireCurrentFrameData=true
      ComponentTags []
      FilteredBones [BodyPieceTop_l, BodyPieceTop_r, BodyPieceBottom_l, BodyPieceBottom_r]
      FilteredSockets []
      MeshUserParameter {parameter {name=None, typeDefHandle {registeredTypeIndex=1}, varData []}}
      SamplingRegions []
     }}
   )
   &InitialMeshOrientation(
    Rotation=&RandomRangeVector(Minimum=<0 0 0>, Maximum=<1 1 1>)
    Use Orientation Vector=true
    Use Rotation Vector=true
   )
  ]
  particleUpdateScript [
   &ParticleState()
   &DynamicMaterialParameters(
    Index 0 Param 1=&RandomRangeFloat(Minimum=0, Maximum=1)
    Index 0 Param 2=1
    Index 0 Param 3=&FloatFromCurve(FloatCurve=DataInterface() {propertyValues {Curve {keys [{interpMode=RCIM_Cubic, tangentMode=RCTM_User, time=0.2, value=1}, {interpMode=RCIM_Cubic, tangentMode=RCTM_User, time=1, value=2}]}}})
    Index 0 Param 4=1
    Module.Param0WriteEnabled=true
    Module.Param1WriteEnabled=true
    Module.Param2WriteEnabled=true
    Module.Param3WriteEnabled=true
   )
   &ScaleColor(
    Scale Alpha=&FloatFromCurve(FloatCurve=DataInterface() {propertyValues {Curve {keys [{interpMode=RCIM_Cubic, tangentMode=RCTM_User, time=0, value=1}, {interpMode=RCIM_Cubic, tangentMode=RCTM_User, time=1, value=0.1}]}}})
    Scale RGB=&MakeVectorFromLinearColorRGB(Color=&LerpLinearColors(StartColor=$Particles.ColorStart, EndColor=$System.ColorOverheat, LerpFactor=$System.ColorOverheatCurve))
    ScaleA=true
    ScaleRGB=true
   )
   &ScaleMeshSize(Scale Factor=&VectorFromFloat(Value=&FloatFromCurve(FloatCurve=DataInterface() {propertyValues {Curve {keys [<at 0 0.2>,<at 0.1 1>,<at 1 0.4>]}}})))
  ]
  rendererClasses [&NiagaraMeshRendererProperties]
  renderers [{rendererClass=&NiagaraMeshRendererProperties, rendererIndex=0}]
 }
}
summary {
 systemName=NS_Powerup_MiniGun_Idle
 emitters [
  {emitterName=ElectricSprites, rendererClasses [&NiagaraSpriteRendererProperties]}
  {emitterName=ElectricityBoneFlow, rendererClasses [&NiagaraRibbonRendererProperties]}
  {emitterName=ElectricityBoneRear, rendererClasses [&NiagaraRibbonRendererProperties]}
  {emitterName=Glow, rendererClasses [&NiagaraSpriteRendererProperties]}
  {emitterName=ShapesSmall, rendererClasses [&NiagaraMeshRendererProperties]}
 ]
 userVariables [
  User.Overheat {type=NiagaraFloat, defaultValue=0, description=""}
  User.SkeletalMesh {
   type=NiagaraDataInterfaceSkeletalMesh
   defaultValue {
    struct=NiagaraExt_VariableValue_DataInterface
    value {dataInterface=&self:NiagaraDataInterfaceSkeletalMesh_0}
   }
  }
  User.SkeletalMesh2 {
   type=NiagaraDataInterfaceSkeletalMesh
   defaultValue {
    struct=NiagaraExt_VariableValue_DataInterface
    value {dataInterface=&self:NiagaraDataInterfaceSkeletalMesh_4}
   }
  }
  User.SkeletalMesh3 {
   type=NiagaraDataInterfaceSkeletalMesh
   defaultValue {
    struct=NiagaraExt_VariableValue_DataInterface
    value {dataInterface=&self:NiagaraDataInterfaceSkeletalMesh_3}
   }
  }
  User.SkeletalMesh_Overheat {
   type=NiagaraDataInterfaceSkeletalMesh
   defaultValue {
    struct=NiagaraExt_VariableValue_DataInterface
    value {dataInterface=&self:NiagaraDataInterfaceSkeletalMesh_2}
   }
  }
  User.SkeletalMesh_Rear {
   type=NiagaraDataInterfaceSkeletalMesh
   defaultValue {
    struct=NiagaraExt_VariableValue_DataInterface
    value {dataInterface=&self:NiagaraDataInterfaceSkeletalMesh_1}
   }
  }
 ]
}
