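# Niagara effect for the minigun power-up's overheat, layering purple-magenta electricity, glows, mesh shapes and smoke on the gun's skeletal mesh.
# Corona, Smoke and ShapesBurst form bursts, while Glow and ShapesSmall spawn continuously, gated by User.Overheat.
# BoneFlow ribbons on three mesh inputs and glows on six front barrel bones spawn by System.OverheatCurve probability.
# These ribbons, glows and small shapes blend toward System.ColorOverheat; two variant emitters are disabled.
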NiagaraSystem /CorruptionItems/Gameplay/Items/PowerUp/Effects/NS_Powerup_MiniGun_Overheat
dependencies {
 usedDataInterfaces [&NiagaraDataInterfaceCurve, &NiagaraDataInterfaceColorCurve]
 usedDynamicInputs []
 usedModules [
  &SystemState
  &self:SystemSpawnScript.SystemScriptSource.SystemScriptGraph.NiagaraNodeAssignment_3.SetVariables_#1
  &self:SystemSpawnScript.SystemScriptSource.SystemScriptGraph.NiagaraNodeAssignment_0.SetVariables_#2
  &self:SystemSpawnScript.SystemScriptSource.SystemScriptGraph.NiagaraNodeAssignment_1.SetVariables_#3
  &self:SystemSpawnScript.SystemScriptSource.SystemScriptGraph.NiagaraNodeAssignment_2.SetVariables_#4
 ]
}
emitters {
 Corona {
  emitterUpdateScript [
   &EmitterState(Life Cycle Mode=::NewEnumerator1, Loop Behavior=::NewEnumerator1, Loop Duration=1, UseLoopDelay=false)
   &SpawnBurst_Instantaneous(Spawn Count=1, Use Spawn Probability=true)
  ]
  particleSpawnScript [
   &InitializeParticle(
    Color=<rgba 6 0.3618 5.438 1>
    Lifetime=0.12
    Sprite Size Mode=::NewEnumerator3
    Uniform Sprite Size=150
    UsePositionOffset=false
    Write Material Random=true
   )
   &SkeletalMeshLocation(
    Bone Sampling Mode=::NewEnumerator4
    Interpolate Mesh Sampling=false
    MaskSpawnGroup=false
    Sampled Position Offset=&RandomRangeVector(Minimum=<-3 -3 -3>, Maximum=<3 3 3>)
    Skeletal Mesh=$User.SkeletalMesh
   )
  ]
  particleUpdateScript [
   &ParticleState()
   &ScaleColor(
    Scale Alpha=&FloatFromCurve(FloatCurve=DataInterface() {propertyValues {Curve {keys [<at 0.2 1 RCIM_Cubic>,<at 1 0>]}}})
    Scale RGB=&VectorFromFloat(Value=1)
    ScaleA=true
    ScaleRGB=true
   )
   &DynamicMaterialParameters(
    Index 0 Param 1=&FloatFromCurve(FloatCurve=DataInterface() {propertyValues {Curve {keys [<at 0 1.5 RCIM_Cubic>, {interpMode=RCIM_Cubic, leaveTangent=-2.4626, time=1, value=0.3}]}}})
    Index 0 Param 2=0.4
    Index 0 Param 3=1
    Index 0 Param 4=1
    Module.Param0WriteEnabled=true
    Module.Param1WriteEnabled=true
    Module.Param2WriteEnabled=true
    Module.Param3WriteEnabled=true
   )
   &CameraOffset(Camera Offset Amount=10)
   &ScaleSpriteSize(Uniform Curve Sprite Scale=DataInterface() {propertyValues {Curve {keys [
        {interpMode=RCIM_Cubic, leaveTangent=2.1646, tangentMode=RCTM_Break, time=0, value=0.1}
        {interpMode=RCIM_Cubic, tangentMode=RCTM_Break, time=1, value=1}
       ]}}})
  ]
  rendererClasses [&NiagaraSpriteRendererProperties]
  renderers [{rendererClass=&NiagaraSpriteRendererProperties}]
 }
 ElectricityBoneFlow2 {
  emitterSpawnScript [
   /SetVariables(Emitter.RandomSeed:NiagaraInt32=/Niagara/DynamicInputs/Random/RandomInt(Randomness Mode:ENiagaraRandomnessMode=::NewEnumerator0, Random Seed:NiagaraInt32=0, Fixed Random Seed:NiagaraBool=false)) {isSetParameters=true, module=SetVariables_#5}
   /Game/Effects/Niagara/Modules/Partition/FN_PartitionAttributes(Number of Bone Sets=1, Particles Per Partition=6)
  ]
  emitterUpdateScript [
   &EmitterState(Inactive Response=::NewEnumerator1, Life Cycle Mode=::NewEnumerator1, Loop Duration=0.4, UseLoopDelay=false)
   &SpawnBurst_Instantaneous(
    Spawn Count=$Emitter.FN_PartitionAttributes.Spawn Total
    Spawn Probability=$System.OverheatCurve
    Spawn Time=&RandomRangeFloat(Minimum=0, Maximum=0.1)
    Use Spawn Probability=true
   )
  ]
  particleSpawnScript [
   /Game/Effects/Niagara/Modules/Partition/FN_Particle_partition_v5/FN_Particle_Partition_v5(Enable Ribbon ID=true, Pattern:ENiagaraPartitionWaveForm=::NewEnumerator1, Seed:NiagaraInt32=$Emitter.RandomSeed)
   /SetVariables(Particles.ColorStart:LinearColor=&MultiplyLinearColor(
     Linear Color=&RandomRangeLinearColor(Minimum=<rgba 0.0221 0 0.3 1>, Maximum=<rgba 1 0.0603 0.9063 1>)
     Scale Factor=&MakeFloat4FromVectorAndFloat("Vector (RGB)"=&VectorFromFloat(Value=1), "Float (Alpha)"=1)
    )) {isSetParameters=true}
   &InitializeParticle(Lifetime=0.3, UsePositionOffset=false, Write Material Random=true)
   &InitializeRibbon(
    Lifetime Max=0.75
    Lifetime Min=0.2
    Lifetime Mode:ENiagara_LifetimeMode=::NewEnumerator1
    Ribbon Width=&RandomRangeFloat(Minimum=6, Maximum=13)
    UsePositionOffset=false
    Write Material Random=true
   )
   /Game/Effects/Niagara/Systems/Prototype/Tests/MusicDrivenCCPM/FN_Initialize_BoneFlow_Attributes(
    Max T Rate=1
    MaxTScale=1
    Number of Bone Sets=1
    Radius Multiplier:NiagaraFloat=&RandomRangeFloat(Minimum=-10, Maximum=10)
    RadiusMax=1
    RadiusMin=0.2
    Rate=1
   )
   &DynamicMaterialParameters(
    Index 0 Param 2=&FloatFromCurve(FloatCurve=DataInterface() {propertyValues {Curve {keys [<at 0 0>,<at 1 -0.8>]}}})
    Index 0 Param 3=&RandomRangeFloat(Minimum=0.3, Maximum=0.4)
    Module.Param0WriteEnabled=false
    Module.Param1WriteEnabled=true
    Module.Param2WriteEnabled=true
    Module.Param3WriteEnabled=false
   )
  ]
  particleUpdateScript [
   &ParticleState()
   &FN_BoneFlow(
    Angle:NiagaraFloat=$Particles.FN_Initialize_BoneFlow_Attributes.TScale
    BoneID:NiagaraInt32=$Particles.Partition.ID
    Enable Position Offsets=true
    Height:NiagaraFloat=$Particles.Initial.Partition.NormalizedID
    Radius:NiagaraFloat=$Particles.Initial.FN_Initialize_BoneFlow_Attributes.Radius
    Set Material Dynamic=false
    SkeletalMesh:NiagaraDataInterfaceSkeletalMesh=$User.SkeletalMesh2
   )
   &ScaleColor(
    Scale Mode=::NewEnumerator1
    Scale RGBA=&Vector4FromCurve(Vector4Curve=DataInterface() {propertyValues {
       AlphaCurve {keys [<at 0 0>,<at 0.2262 1>,<at 0.4885 1>,<at 1 1>]}
       BlueCurve {keys [<at 0.3099 20>,<at 0.3685 10>,<at 0.9991 2>]}
       GreenCurve {keys [<at 0.3099 20>,<at 0.3685 10>,<at 0.9991 2>]}
       RedCurve {keys [<at 0.3099 20>,<at 0.3685 10>,<at 0.9991 2>]}
      }})
    ScaleRGBA=true
   ) {module=ScaleColor001}
   &ScaleRibbonWidth(Ribbon Width Scale=2)
   &ScaleColor(
    Scale Alpha=&FloatFromCurve(
     FloatCurve=DataInterface() {propertyValues {Curve {keys [<at 0 0>,<at 0.1 1>,<at 0.2 1>,<at 0.3 0>,<at 0.7 0>,<at 0.8 1>,<at 0.9 1>,<at 1 0>]}}}
     CurveIndex=$Particles.Partition.NormalizedID
    )
    Scale RGB=&MakeVectorFromLinearColorRGB(Color=&LerpLinearColors(StartColor=$Particles.ColorStart, EndColor=$System.ColorOverheat, LerpFactor=$System.ColorOverheatCurve))
    ScaleA=true
    ScaleRGB=true
   )
  ]
  rendererClasses [&NiagaraRibbonRendererProperties]
  renderers [{rendererClass=&NiagaraRibbonRendererProperties}]
 }
 ElectricityBoneFlow3 {
  emitterSpawnScript [
   /SetVariables(Emitter.RandomSeed:NiagaraInt32=&RandomInt(Randomness Mode:ENiagaraRandomnessMode=::NewEnumerator0, Random Seed:NiagaraInt32=0, Fixed Random Seed:NiagaraBool=false)) {isSetParameters=true, module=SetVariables_#5}
   &FN_PartitionAttributes(Number of Bone Sets=1, Particles Per Partition=6)
  ]
  emitterUpdateScript [
   &EmitterState(Inactive Response=::NewEnumerator1, Life Cycle Mode=::NewEnumerator1, Loop Duration=0.4, UseLoopDelay=false)
   &SpawnBurst_Instantaneous(
    Spawn Count=$Emitter.FN_PartitionAttributes.Spawn Total
    Spawn Probability=$System.OverheatCurve
    Spawn Time=&RandomRangeFloat(Minimum=0, Maximum=0.1)
    Use Spawn Probability=true
   )
  ]
  particleSpawnScript [
   &FN_Particle_Partition_v5(Enable Ribbon ID=true, Pattern:ENiagaraPartitionWaveForm=::NewEnumerator1, Seed:NiagaraInt32=$Emitter.RandomSeed)
   /SetVariables(Particles.ColorStart:LinearColor=&MultiplyLinearColor(
     Linear Color=&RandomRangeLinearColor(Minimum=<rgba 0.0221 0 0.3 1>, Maximum=<rgba 1 0.0603 0.9063 1>)
     Scale Factor=&MakeFloat4FromVectorAndFloat("Vector (RGB)"=&VectorFromFloat(Value=1), "Float (Alpha)"=1)
    )) {isSetParameters=true}
   &InitializeParticle(Lifetime=0.3, UsePositionOffset=false, Write Material Random=true)
   &InitializeRibbon(
    Lifetime Max=0.75
    Lifetime Min=0.2
    Lifetime Mode:ENiagara_LifetimeMode=::NewEnumerator1
    Ribbon Width=&RandomRangeFloat(Minimum=6, Maximum=13)
    UsePositionOffset=false
    Write Material Random=true
   )
   &FN_Initialize_BoneFlow_Attributes(
    Number of Bone Sets=1
    Radius Multiplier:NiagaraFloat=&RandomRangeFloat(Minimum=-10, Maximum=10)
    RadiusMax=1
    RadiusMin=0.2
    Rate=1
   )
   &DynamicMaterialParameters(
    Index 0 Param 2=&FloatFromCurve(FloatCurve=DataInterface() {propertyValues {Curve {keys [<at 0 0>,<at 1 -0.8>]}}})
    Index 0 Param 3=&RandomRangeFloat(Minimum=0.3, Maximum=0.4)
    Module.Param0WriteEnabled=false
    Module.Param1WriteEnabled=true
    Module.Param2WriteEnabled=true
    Module.Param3WriteEnabled=false
   )
  ]
  particleUpdateScript [
   &ParticleState()
   &FN_BoneFlow(
    Angle:NiagaraFloat=$Particles.FN_Initialize_BoneFlow_Attributes.TScale
    BoneID:NiagaraInt32=$Particles.Partition.ID
    Enable Position Offsets=true
    Height:NiagaraFloat=$Particles.Initial.Partition.NormalizedID
    Radius:NiagaraFloat=$Particles.Initial.FN_Initialize_BoneFlow_Attributes.Radius
    Set Material Dynamic=false
    SkeletalMesh:NiagaraDataInterfaceSkeletalMesh=$User.SkeletalMesh3
   )
   &ScaleColor(
    Scale Mode=::NewEnumerator1
    Scale RGBA=&Vector4FromCurve(Vector4Curve=DataInterface() {propertyValues {
       AlphaCurve {keys [<at 0 0>,<at 0.2262 1>,<at 0.4885 1>,<at 1 1>]}
       BlueCurve {keys [<at 0.3099 20>,<at 0.3685 10>,<at 0.9991 2>]}
       GreenCurve {keys [<at 0.3099 20>,<at 0.3685 10>,<at 0.9991 2>]}
       RedCurve {keys [<at 0.3099 20>,<at 0.3685 10>,<at 0.9991 2>]}
      }})
    ScaleRGBA=true
   ) {module=ScaleColor001}
   &ScaleRibbonWidth(Ribbon Width Scale=2)
   &ScaleColor(
    Scale Alpha=&FloatFromCurve(
     FloatCurve=DataInterface() {propertyValues {Curve {keys [<at 0 0>,<at 0.1 1>,<at 0.2 1>,<at 0.3 0>,<at 0.7 0>,<at 0.8 1>,<at 0.9 1>,<at 1 0>]}}}
     CurveIndex=$Particles.Partition.NormalizedID
    )
    Scale RGB=&MakeVectorFromLinearColorRGB(Color=&LerpLinearColors(StartColor=$Particles.ColorStart, EndColor=$System.ColorOverheat, LerpFactor=$System.ColorOverheatCurve))
    ScaleA=true
    ScaleRGB=true
   )
  ]
  rendererClasses [&NiagaraRibbonRendererProperties]
  renderers [{rendererClass=&NiagaraRibbonRendererProperties}]
 }
 ElectricityBoneFlowOverheat {
  emitterSpawnScript [
   /SetVariables(Emitter.RandomSeed:NiagaraInt32=&RandomInt(Randomness Mode:ENiagaraRandomnessMode=::NewEnumerator0, Random Seed:NiagaraInt32=0, Fixed Random Seed:NiagaraBool=false)) {isSetParameters=true, module=SetVariables_#5}
   &FN_PartitionAttributes(Number of Bone Sets=1, Particles Per Partition=6)
  ]
  emitterUpdateScript [
   &EmitterState(Inactive Response=::NewEnumerator1, Life Cycle Mode=::NewEnumerator1, Loop Duration=0.4, UseLoopDelay=false)
   &SpawnBurst_Instantaneous(
    Spawn Count=$Emitter.FN_PartitionAttributes.Spawn Total
    Spawn Probability=$System.OverheatCurve
    Spawn Time=&RandomRangeFloat(Minimum=0, Maximum=0.1)
    Use Spawn Probability=true
   )
  ]
  particleSpawnScript [
   &FN_Particle_Partition_v5(Enable Ribbon ID=true, Pattern:ENiagaraPartitionWaveForm=::NewEnumerator1, Seed:NiagaraInt32=$Emitter.RandomSeed)
   /SetVariables(Particles.ColorStart:LinearColor=&MultiplyLinearColor(
     Linear Color=&RandomRangeLinearColor(Minimum=<rgba 0.0221 0 0.3 1>, Maximum=<rgba 1 0.0603 0.9063 1>)
     Scale Factor=&MakeFloat4FromVectorAndFloat("Vector (RGB)"=&VectorFromFloat(Value=1), "Float (Alpha)"=1)
    )) {isSetParameters=true, module=SetVariables_#6}
   &InitializeParticle(Lifetime=0.3, UsePositionOffset=false, Write Material Random=true)
   &InitializeRibbon(
    Lifetime Max=0.75
    Lifetime Min=0.2
    Lifetime Mode:ENiagara_LifetimeMode=::NewEnumerator1
    Ribbon Width=&RandomRangeFloat(Minimum=6, Maximum=13)
    UsePositionOffset=false
    Write Material Random=true
   )
   &FN_Initialize_BoneFlow_Attributes(
    Number of Bone Sets=1
    Radius Multiplier:NiagaraFloat=&RandomRangeFloat(Minimum=-10, Maximum=10)
    RadiusMax=1
    RadiusMin=0.2
    Rate=1
   )
   &DynamicMaterialParameters(
    Index 0 Param 2=&FloatFromCurve(FloatCurve=DataInterface() {propertyValues {Curve {keys [<at 0 0>,<at 1 -0.8>]}}})
    Index 0 Param 3=&RandomRangeFloat(Minimum=0.3, Maximum=0.4)
    Module.Param0WriteEnabled=false
    Module.Param1WriteEnabled=true
    Module.Param2WriteEnabled=true
    Module.Param3WriteEnabled=false
   )
  ]
  particleUpdateScript [
   &ParticleState()
   &FN_BoneFlow(
    Angle:NiagaraFloat=$Particles.FN_Initialize_BoneFlow_Attributes.TScale
    BoneID:NiagaraInt32=$Particles.Partition.ID
    Enable Position Offsets=true
    Height:NiagaraFloat=$Particles.Initial.Partition.NormalizedID
    Radius:NiagaraFloat=$Particles.Initial.FN_Initialize_BoneFlow_Attributes.Radius
    Set Material Dynamic=false
    SkeletalMesh:NiagaraDataInterfaceSkeletalMesh=$User.SkeletalMesh_Overheat
   )
   &ScaleColor(
    Scale Mode=::NewEnumerator1
    Scale RGBA=&Vector4FromCurve(Vector4Curve=DataInterface() {propertyValues {
       AlphaCurve {keys [<at 0 0>,<at 0.2262 1>,<at 0.4885 1>,<at 1 1>]}
       BlueCurve {keys [<at 0.3099 20>,<at 0.3685 10>,<at 0.9991 2>]}
       GreenCurve {keys [<at 0.3099 20>,<at 0.3685 10>,<at 0.9991 2>]}
       RedCurve {keys [<at 0.3099 20>,<at 0.3685 10>,<at 0.9991 2>]}
      }})
    ScaleRGBA=true
   ) {module=ScaleColor001}
   &ScaleRibbonWidth(Ribbon Width Scale=2)
   &ScaleColor(
    Scale Alpha=&FloatFromCurve(
     FloatCurve=DataInterface() {propertyValues {Curve {keys [<at 0 0>,<at 0.1 1>,<at 0.2 1>,<at 0.3 0>,<at 0.7 0>,<at 0.8 1>,<at 0.9 1>,<at 1 0>]}}}
     CurveIndex=$Particles.Partition.NormalizedID
    )
    Scale RGB=&MakeVectorFromLinearColorRGB(Color=&LerpLinearColors(StartColor=$Particles.ColorStart, EndColor=$System.ColorOverheat, LerpFactor=$System.ColorOverheatCurve))
    ScaleA=true
    ScaleRGB=true
   )
  ]
  rendererClasses [&NiagaraRibbonRendererProperties]
  renderers [{rendererClass=&NiagaraRibbonRendererProperties}]
 }
 ElectricityBoneFlowOverheat001 {
  bEnabled=false
  emitterSpawnScript [
   /SetVariables(Emitter.RandomSeed:NiagaraInt32=&RandomInt(Randomness Mode:ENiagaraRandomnessMode=::NewEnumerator0, Random Seed:NiagaraInt32=0, Fixed Random Seed:NiagaraBool=false)) {isSetParameters=true, module=SetVariables_#5}
   &FN_PartitionAttributes(Number of Bone Sets=1, Particles Per Partition=8)
  ]
  emitterUpdateScript [
   &EmitterState(Inactive Response=::NewEnumerator1, Life Cycle Mode=::NewEnumerator1, Loop Duration=0.4, UseLoopDelay=false)
   &SpawnBurst_Instantaneous(
    Spawn Count=$Emitter.FN_PartitionAttributes.Spawn Total
    Spawn Probability=&MakeCustomFloatFromBool(Bool=&SetBoolByFloatComparison(A=$User.Overheat, B=0), True Float=1, False Float=0)
    Spawn Time=&RandomRangeFloat(Minimum=0, Maximum=0.1)
    Use Spawn Probability=true
   )
  ]
  particleSpawnScript [
   &FN_Particle_Partition_v5(Enable Ribbon ID=true, Pattern:ENiagaraPartitionWaveForm=::NewEnumerator1, Seed:NiagaraInt32=$Emitter.RandomSeed)
   /SetVariables(Particles.ColorStart:LinearColor=&MultiplyLinearColor(
     Linear Color=&RandomRangeLinearColor(Minimum=<rgba 0.0221 0 0.3 1>, Maximum=<rgba 1 0.0603 0.9063 1>)
     Scale Factor=&MakeFloat4FromVectorAndFloat("Vector (RGB)"=&VectorFromFloat(Value=1), "Float (Alpha)"=1)
    )) {isSetParameters=true, module=SetVariables_#6}
   &InitializeParticle(Lifetime=0.3, UsePositionOffset=false, Write Material Random=true)
   &InitializeRibbon(
    Lifetime Max=0.75
    Lifetime Min=0.2
    Lifetime Mode:ENiagara_LifetimeMode=::NewEnumerator1
    Ribbon Width=&RandomRangeFloat(Minimum=6, Maximum=13)
    UsePositionOffset=false
    Write Material Random=true
   )
   &FN_Initialize_BoneFlow_Attributes(
    Number of Bone Sets=1
    Radius Multiplier:NiagaraFloat=&RandomRangeFloat(Minimum=-10, Maximum=10)
    RadiusMax=1
    RadiusMin=0.2
    Rate=1
   )
   &DynamicMaterialParameters(
    Index 0 Param 2=&FloatFromCurve(FloatCurve=DataInterface() {propertyValues {Curve {keys [<at 0 0>,<at 1 -0.8>]}}})
    Index 0 Param 3=&RandomRangeFloat(Minimum=0.3, Maximum=0.4)
    Module.Param0WriteEnabled=false
    Module.Param1WriteEnabled=true
    Module.Param2WriteEnabled=true
    Module.Param3WriteEnabled=false
   )
  ]
  particleUpdateScript [
   &ParticleState()
   &FN_BoneFlow(
    Angle:NiagaraFloat=$Particles.FN_Initialize_BoneFlow_Attributes.TScale
    BoneID:NiagaraInt32=$Particles.Partition.ID
    Enable Position Offsets=true
    Height:NiagaraFloat=$Particles.Initial.Partition.NormalizedID
    Radius:NiagaraFloat=$Particles.Initial.FN_Initialize_BoneFlow_Attributes.Radius
    Set Material Dynamic=false
    SkeletalMesh:NiagaraDataInterfaceSkeletalMesh=$User.SkeletalMesh_Overheat
   )
   &ScaleColor(
    Scale Mode=::NewEnumerator1
    Scale RGBA=&Vector4FromCurve(Vector4Curve=DataInterface() {propertyValues {
       AlphaCurve {keys [<at 0 0>,<at 0.2262 1>,<at 0.4885 1>,<at 1 1>]}
       BlueCurve {keys [<at 0.3099 20>,<at 0.3685 10>,<at 0.9991 2>]}
       GreenCurve {keys [<at 0.3099 20>,<at 0.3685 10>,<at 0.9991 2>]}
       RedCurve {keys [<at 0.3099 20>,<at 0.3685 10>,<at 0.9991 2>]}
      }})
    ScaleRGBA=true
   ) {module=ScaleColor001}
   &ScaleRibbonWidth(Ribbon Width Scale=2)
   &ScaleColor(
    Scale Alpha=&FloatFromCurve(
     FloatCurve=DataInterface() {propertyValues {Curve {keys [<at 0 0>,<at 0.1 1>,<at 0.2 1>,<at 0.3 0>,<at 0.7 0>,<at 0.8 1>,<at 0.9 1>,<at 1 0>]}}}
     CurveIndex=$Particles.Partition.NormalizedID
    )
    Scale RGB=&MakeVectorFromLinearColorRGB(Color=&LerpLinearColors(StartColor=$Particles.ColorStart, EndColor=$System.ColorOverheat, LerpFactor=$System.ColorOverheatCurve))
    ScaleA=true
    ScaleRGB=true
   )
  ]
  rendererClasses [&NiagaraRibbonRendererProperties]
  renderers [{rendererClass=&NiagaraRibbonRendererProperties}]
 }
 Glow {
  emitterUpdateScript [
   &EmitterState()
   &SpawnRate(
    Spawn Probability=&MakeCustomFloatFromBool(Bool=&SetBoolByFloatComparison(A=$User.Overheat, B=0), True Float=1, False Float=0)
    SpawnRate=5
    Use Spawn Probability=true
   )
  ]
  particleSpawnScript [
   /SetVariables(Particles.ColorStart:LinearColor=<rgba 1 0.0603 0.9063 1>) {isSetParameters=true}
   &InitializeParticle(
    Lifetime Max=0.5
    Lifetime Min=0.2
    Lifetime Mode=::NewEnumerator1
    Sprite Rotation Mode=::NewEnumerator2
    Sprite Size Mode=::NewEnumerator1
    Uniform Sprite Size Max=90
    Uniform Sprite Size Min=50
    UsePositionOffset=false
    Write Material Random=true
   )
   &SkeletalMeshLocation(Bone Sampling Mode=::NewEnumerator4, Interpolate Mesh Sampling=false, MaskSpawnGroup=false, Skeletal Mesh=$User.SkeletalMesh)
  ]
  particleUpdateScript [
   &ParticleState()
   &ScaleSpriteSize(Uniform Curve Sprite Scale=DataInterface() {propertyValues {Curve {keys [
        {interpMode=RCIM_Cubic, tangentMode=RCTM_User, time=0, value=0.2}
        {interpMode=RCIM_Cubic, tangentMode=RCTM_Break, time=0.3, value=1}
        {arriveTangent=-1.2264, interpMode=RCIM_Cubic, leaveTangent=-1.2264, tangentMode=RCTM_User, time=1, value=0.6}
       ]}}})
   &ScaleColor(
    Scale RGB=&MakeVectorFromLinearColorRGB(Color=&LerpLinearColors(StartColor=$Particles.ColorStart, EndColor=$System.ColorOverheat, LerpFactor=$System.ColorOverheatCurve))
    ScaleA=false
    ScaleRGB=true
   )
  ]
  rendererClasses [&NiagaraSpriteRendererProperties]
  renderers [{rendererClass=&NiagaraSpriteRendererProperties}]
 }
 GlowContactPoints {
  emitterUpdateScript [
   &EmitterState(Inactive Response=::NewEnumerator1, Life Cycle Mode=::NewEnumerator1, Loop Duration=0.6, UseLoopDelay=false)
   &SpawnRate(Spawn Probability=$System.OverheatCurve, SpawnRate=30, Use Spawn Probability=true)
  ]
  particleSpawnScript [
   &FN_SetSpawnGroup(Number of Spawn Groups=6)
   /SetVariables(Particles.ColorStart:LinearColor=<rgba 1 0.0603 0.9063 1>) {isSetParameters=true, module=SetVariables_#7}
   &InitializeParticle(
    Lifetime Max=0.25
    Lifetime Min=0.1
    Lifetime Mode=::NewEnumerator1
    Sprite Size Mode=::NewEnumerator1
    Uniform Sprite Size Max=28
    Uniform Sprite Size Min=20
    UsePositionOffset=false
    Write Material Random=true
   )
   &CameraOffset(Camera Offset Amount=10)
  ]
  particleUpdateScript [
   &ParticleState()
   &SkeletalMeshLocation(
    Bone / Socket Index=$Particles.SpawnGroup
    Bone Sampling Mode=::NewEnumerator4
    MaskSpawnGroup=false
    Sampled Position Offset=&RandomRangeVector(Minimum=<-2 -2 -2>, Maximum=<2 2 2>)
    Skeletal Mesh=DataInterface() {propertyValues {
      DefaultMesh=null
      ExcludeBoneName=root
      OptionalComponentClass=null
      PreviewMesh=/CorruptionItems/Gameplay/Items/PowerUp/LMG/Meshes/Power_Up_Minigun_Gun
      SkinningMode=SkinOnTheFly
      SoftSourceActor=null
      SourceMode=Default
      UvSetIndex=0
      WholeMeshLOD=-1
      bExcludeBone=false
      bReadDeformedGeometry=true
      bRequireCurrentFrameData=true
      ComponentTags []
      FilteredBones [BarrelFront1,BarrelFront2,BarrelFront3,BarrelFront4,BarrelFront5,BarrelFront6]
      FilteredSockets []
      MeshUserParameter {parameter {name=None, typeDefHandle {registeredTypeIndex=1}, varData []}}
      SamplingRegions []
     }}
   )
   &ScaleColor(
    Scale Alpha=&FloatFromCurve(
     FloatCurve=DataInterface() {propertyValues {Curve {keys [<at 0.6 1 RCIM_Cubic>, {arriveTangent=-5.6404, interpMode=RCIM_Cubic, tangentMode=RCTM_Break, time=1, value=0}]}}}
     Scale Curve=0.2
    )
    Scale RGB=&MakeVectorFromLinearColorRGB(Color=&LerpLinearColors(StartColor=$Particles.ColorStart, EndColor=$System.ColorOverheat, LerpFactor=$System.ColorOverheatCurve))
    ScaleA=true
    ScaleRGB=true
   )
   /SetVariables(Particles.Test:NiagaraFloat=$User.Overheat) {isSetParameters=true, module=SetVariables_#8}
  ]
  rendererClasses [&NiagaraSpriteRendererProperties]
  renderers [{rendererClass=&NiagaraSpriteRendererProperties}]
 }
 GlowContactPoints001 {
  bEnabled=false
  emitterUpdateScript [
   &EmitterState(Inactive Response=::NewEnumerator1, Life Cycle Mode=::NewEnumerator1, Loop Duration=0.6, UseLoopDelay=false)
   &SpawnBurst_Instantaneous(
    Spawn Count=6
    Spawn Probability=&MakeCustomFloatFromBool(Bool=&SetBoolByFloatComparison(A=$User.Overheat, B=0), True Float=1, False Float=0)
    Use Spawn Probability=true
    UseLoopCountLimit=false
   ) {enabled=false}
   &SpawnRate(
    Spawn Probability=&MakeCustomFloatFromBool(Bool=&SetBoolByFloatComparison(A=$User.Overheat, B=0.15), True Float=1, False Float=0)
    SpawnRate=30
    Use Spawn Probability=true
   )
  ]
  particleSpawnScript [
   &FN_SetSpawnGroup(Number of Spawn Groups=6)
   /SetVariables(Particles.ColorStart:LinearColor=<rgba 1 0.0603 0.9063 1>) {isSetParameters=true, module=SetVariables_#7}
   &InitializeParticle(
    Lifetime Max=0.25
    Lifetime Min=0.1
    Lifetime Mode=::NewEnumerator1
    Sprite Size Mode=::NewEnumerator1
    Uniform Sprite Size Max=28
    Uniform Sprite Size Min=20
    UsePositionOffset=false
    Write Material Random=true
   )
   &CameraOffset(Camera Offset Amount=10)
  ]
  particleUpdateScript [
   &ParticleState()
   &SkeletalMeshLocation(
    Bone / Socket Index=$Particles.SpawnGroup
    Bone Sampling Mode=::NewEnumerator4
    MaskSpawnGroup=false
    Sampled Position Offset=&RandomRangeVector(Minimum=<-2 -2 -2>, Maximum=<2 2 2>)
    Skeletal Mesh=DataInterface() {propertyValues {
      DefaultMesh=null
      ExcludeBoneName=root
      OptionalComponentClass=null
      PreviewMesh=&Power_Up_Minigun_Gun
      SkinningMode=SkinOnTheFly
      SoftSourceActor=null
      SourceMode=Default
      UvSetIndex=0
      WholeMeshLOD=-1
      bExcludeBone=false
      bReadDeformedGeometry=true
      bRequireCurrentFrameData=true
      ComponentTags []
      FilteredBones [BarrelFront1,BarrelFront2,BarrelFront3,BarrelFront4,BarrelFront5,BarrelFront6]
      FilteredSockets []
      MeshUserParameter {parameter {name=None, typeDefHandle {registeredTypeIndex=1}, varData []}}
      SamplingRegions []
     }}
   )
   &ScaleColor(
    Scale Alpha=&FloatFromCurve(
     FloatCurve=DataInterface() {propertyValues {Curve {keys [<at 0.6 1 RCIM_Cubic>, {arriveTangent=-5.6404, interpMode=RCIM_Cubic, tangentMode=RCTM_Break, time=1, value=0}]}}}
     Scale Curve=0.2
    )
    Scale RGB=&MakeVectorFromLinearColorRGB(Color=&LerpLinearColors(StartColor=$Particles.ColorStart, EndColor=$System.ColorOverheat, LerpFactor=$System.ColorOverheatCurve))
    ScaleA=true
    ScaleRGB=true
   )
   /SetVariables(Particles.Test:NiagaraFloat=$User.Overheat) {isSetParameters=true, module=SetVariables_#8}
  ]
  rendererClasses [&NiagaraSpriteRendererProperties]
  renderers [
   {rendererClass=&NiagaraSpriteRendererProperties}
   {rendererClass=&NiagaraSpriteRendererProperties, rendererIndex=1}
  ]
 }
 ShapesBurst {
  emitterUpdateScript [
   &EmitterState()
   &SpawnBurst_Instantaneous(Spawn Count=6, Use Spawn Probability=false, UseLoopCountLimit=false)
  ]
  particleSpawnScript [
   &InitializeParticle(
    Lifetime Max=0.4
    Lifetime Min=0.2
    Lifetime Mode=::NewEnumerator1
    Mesh Scale Mode=::NewEnumerator1
    Mesh Uniform Scale Max=5
    Mesh Uniform Scale Min=2
    UsePositionOffset=false
    Write Material Random=true
   )
   &SkeletalMeshLocation(
    Bone Sampling Mode=::NewEnumerator4
    Interpolate Mesh Sampling=false
    MaskSpawnGroup=false
    Sampled Position Offset=&RandomRangeVector(Minimum=<-2 -2 -2>, Maximum=<2 2 2>)
    Skeletal Mesh=$User.SkeletalMesh
   )
   &InitialMeshOrientation(
    Rotation=&RandomRangeVector(Minimum=<0 0 0>, Maximum=<1 1 1>)
    Use Orientation Vector=true
    Use Rotation Vector=true
   )
   &AddVelocityFromPoint(
    Module.Use Falloff Distance=false
    Velocity Origin:NiagaraPosition=$Output.SkeletalMeshLocation.SampledPositionWithoutOffset
    Velocity Strength=&RandomRangeFloat(Minimum=25, Maximum=300)
   )
  ]
  particleUpdateScript [
   &Drag(Use Linear Drag=true, Use Rotational Drag=false)
   &SolveForcesAndVelocity(Clamp Velocity=false, Limit Acceleration=false)
   &ParticleState()
   &DynamicMaterialParameters(
    Index 0 Param 1=&RandomRangeFloat(Minimum=0, Maximum=1)
    Index 0 Param 2=1
    Index 0 Param 3=&FloatFromCurve(FloatCurve=DataInterface() {propertyValues {Curve {keys [{interpMode=RCIM_Cubic, tangentMode=RCTM_User, time=0.2, value=1}, {interpMode=RCIM_Cubic, tangentMode=RCTM_User, time=1, value=2}]}}})
    Index 0 Param 4=1
    Module.Param0WriteEnabled=true
    Module.Param1WriteEnabled=true
    Module.Param2WriteEnabled=true
    Module.Param3WriteEnabled=true
   )
   &ScaleColor(
    Scale Alpha=&FloatFromCurve(FloatCurve=DataInterface() {propertyValues {Curve {keys [{interpMode=RCIM_Cubic, tangentMode=RCTM_User, time=0, value=1}, {interpMode=RCIM_Cubic, tangentMode=RCTM_User, time=1, value=0.1}]}}})
    Scale RGB=&MakeVectorFromLinearColorRGB(Color=&LerpLinearColors(
      StartColor=<rgba 6 0.3618 5.438 1>
      EndColor=<rgba 0.7099 0 2 1>
      LerpFactor=&FloatFromCurve(FloatCurve=DataInterface() {propertyValues {Curve {keys [{interpMode=RCIM_Cubic, tangentMode=RCTM_User, time=0.1, value=0}, {interpMode=RCIM_Cubic, tangentMode=RCTM_User, time=0.8, value=1}]}}})
     ))
    ScaleA=true
    ScaleRGB=true
   )
   &ScaleMeshSize(Scale Factor=&VectorFromFloat(Value=&FloatFromCurve(FloatCurve=DataInterface() {propertyValues {Curve {keys [<at 0 0.2>,<at 0.1 1>,<at 1 0.4>]}}})))
  ]
  rendererClasses [&NiagaraMeshRendererProperties]
  renderers [{rendererClass=&NiagaraMeshRendererProperties, rendererIndex=0}]
 }
 ShapesSmall {
  emitterUpdateScript [
   &EmitterState()
   &SpawnRate(
    Spawn Probability=&MakeCustomFloatFromBool(Bool=&SetBoolByFloatComparison(A=$User.Overheat, B=0), True Float=1, False Float=0)
    SpawnRate=25
    Use Spawn Probability=true
   )
  ]
  particleSpawnScript [
   /SetVariables(Particles.ColorStart:LinearColor=<rgba 2 0.1206 1.8127 1>) {isSetParameters=true}
   &InitializeParticle(
    Lifetime Max=0.15
    Lifetime Min=0.08
    Lifetime Mode=::NewEnumerator1
    Mesh Scale Mode=::NewEnumerator1
    Mesh Uniform Scale Max=3
    UsePositionOffset=false
    Write Material Random=true
   )
   &SkeletalMeshLocation(
    Bone Sampling Mode=::NewEnumerator4
    Interpolate Mesh Sampling=false
    MaskSpawnGroup=false
    Sampled Position Offset=&RandomRangeVector(Minimum=<-2 -2 -2>, Maximum=<2 2 2>)
    Skeletal Mesh=$User.SkeletalMesh
   )
   &InitialMeshOrientation(
    Rotation=&RandomRangeVector(Minimum=<0 0 0>, Maximum=<1 1 1>)
    Use Orientation Vector=true
    Use Rotation Vector=true
   )
  ]
  particleUpdateScript [
   &ParticleState()
   &DynamicMaterialParameters(
    Index 0 Param 1=&RandomRangeFloat(Minimum=0, Maximum=1)
    Index 0 Param 2=1
    Index 0 Param 3=&FloatFromCurve(FloatCurve=DataInterface() {propertyValues {Curve {keys [{interpMode=RCIM_Cubic, tangentMode=RCTM_User, time=0.2, value=1}, {interpMode=RCIM_Cubic, tangentMode=RCTM_User, time=1, value=2}]}}})
    Index 0 Param 4=1
    Module.Param0WriteEnabled=true
    Module.Param1WriteEnabled=true
    Module.Param2WriteEnabled=true
    Module.Param3WriteEnabled=true
   )
   &ScaleColor(
    Scale Alpha=&FloatFromCurve(FloatCurve=DataInterface() {propertyValues {Curve {keys [{interpMode=RCIM_Cubic, tangentMode=RCTM_User, time=0, value=1}, {interpMode=RCIM_Cubic, tangentMode=RCTM_User, time=1, value=0.1}]}}})
    Scale RGB=&MakeVectorFromLinearColorRGB(Color=&LerpLinearColors(StartColor=$Particles.ColorStart, EndColor=$System.ColorOverheat, LerpFactor=$System.ColorOverheatCurve))
    ScaleA=true
    ScaleRGB=true
   )
   &ScaleMeshSize(Scale Factor=&VectorFromFloat(Value=&FloatFromCurve(FloatCurve=DataInterface() {propertyValues {Curve {keys [<at 0 0.2>,<at 0.1 1>,<at 1 0.4>]}}})))
  ]
  rendererClasses [&NiagaraMeshRendererProperties]
  renderers [{rendererClass=&NiagaraMeshRendererProperties, rendererIndex=0}]
 }
 Smoke {
  emitterUpdateScript [
   &EmitterState()
   &SpawnBurst_Instantaneous(Spawn Count=&UniformRangedInt(Minimum=5, Maximum=8), Use Spawn Probability=false)
  ]
  particleSpawnScript [
   /SetVariables(Particles.InitSpriteRotRate=&RandomRangeFloat(Minimum=-1, Maximum=1)) {isSetParameters=true}
   &InitializeParticle(
    Color Maximum=<rgba 12 0.7236 10.876 1>
    Color Minimum=<rgba 0.0221 0 0.3 1>
    Color Mode=::NewEnumerator2
    Lifetime Max=0.35
    Lifetime Min=0.2
    Lifetime Mode=::NewEnumerator1
    Sprite Rotation Mode=::NewEnumerator2
    Sprite Size Mode=::NewEnumerator1
    Uniform Sprite Size Max=90
    Uniform Sprite Size Min=60
    UsePositionOffset=false
    Write Material Random=true
   )
   &SkeletalMeshLocation(
    Bone Sampling Mode=::NewEnumerator4
    Interpolate Mesh Sampling=false
    MaskSpawnGroup=false
    Sampled Position Offset=&RandomRangeVector(Minimum=<-2 -2 -2>, Maximum=<2 2 2>)
    Skeletal Mesh=$User.SkeletalMesh
   )
   &AddVelocityFromPoint(
    Module.Use Falloff Distance=false
    Velocity Origin:NiagaraPosition=$Output.SkeletalMeshLocation.SampledPositionWithoutOffset
    Velocity Strength=&RandomRangeFloat(Minimum=100, Maximum=500)
   )
   /NiagaraFN/Sprite/FN_RandomizeSpriteUV()
  ]
  particleUpdateScript [
   &ParticleState()
   &Drag(Use Linear Drag=true, Use Rotational Drag=false)
   &ScaleColor(
    Scale Alpha=&FloatFromCurve(
     FloatCurve=DataInterface() {propertyValues {Curve {keys [
         {arriveTangent=6.5041, interpMode=RCIM_Cubic, leaveTangent=6.5041, tangentMode=RCTM_User, time=0, value=0}
         {interpMode=RCIM_Cubic, tangentMode=RCTM_User, time=0.15, value=1}
         <at 1 0>
        ]}}}
     Scale Curve=&RandomRangeFloat(Minimum=0.15, Maximum=0.3)
    )
    Scale RGB=&VectorFromFloat(Value=&FloatFromCurve(FloatCurve=DataInterface() {propertyValues {Curve {keys [<at 0 1>,<at 1 0>]}}}))
    ScaleA=true
    ScaleRGB=true
   )
   &ScaleSpriteSize(Scale Factor=&Vector2DFromFloat(Value=&FloatFromCurve(FloatCurve=DataInterface() {propertyValues {Curve {keys [
          {arriveTangent=0.4478, interpMode=RCIM_Cubic, leaveTangent=0.4478, tangentMode=RCTM_User, time=0, value=0.5}
          {arriveTangent=0.9127, interpMode=RCIM_Cubic, leaveTangent=0.9127, tangentMode=RCTM_User, time=0.5, value=1.125}
          {arriveTangent=2.1782, interpMode=RCIM_Cubic, leaveTangent=2.1782, tangentMode=RCTM_User, time=1.0008, value=1.6042}
         ]}}})))
   &DynamicMaterialParameters(
    Index 0 Param 1=&Add_Float(
     A=&RandomRangeFloat(Minimum=0, Maximum=1)
     B=&FloatFromCurve(FloatCurve=DataInterface() {propertyValues {Curve {keys [<at 0 1>,<at 1 0>]}}}, Scale Curve=0.05)
    )
    Index 0 Param 3=&FloatFromCurve(FloatCurve=DataInterface() {propertyValues {Curve {keys [{arriveTangent=-22.7009, interpMode=RCIM_Cubic, leaveTangent=0.0563, tangentMode=RCTM_Break, time=0, value=2}, <at 0.5 0 RCIM_Cubic>]}}})
    Index 0 Param 4=&FloatFromCurve(FloatCurve=DataInterface() {propertyValues {Curve {keys [{arriveTangent=-3.2497, interpMode=RCIM_Cubic, leaveTangent=-3.2497, tangentMode=RCTM_User, time=0, value=1}, <at 1 0 RCIM_Cubic>]}}})
    Index 1 Param 1=&FloatFromCurve(FloatCurve=DataInterface() {propertyValues {Curve {keys [<at 0 1>, {arriveTangent=4.2257, interpMode=RCIM_Cubic, leaveTangent=4.2257, tangentMode=RCTM_User, time=1, value=1}]}}})
    Index 1 Param 2=&FloatFromCurve(FloatCurve=DataInterface() {propertyValues {Curve {keys [<at 0 0>]}}})
    Index 1 Param 3=&FloatFromCurve(
     FloatCurve=DataInterface() {propertyValues {Curve {keys [<at 0 0.05 RCIM_Cubic>, {arriveTangent=0.0641, interpMode=RCIM_Cubic, leaveTangent=6.3917, tangentMode=RCTM_Break, time=0.5, value=1}]}}}
     Scale Curve=4
    )
    Index 1 Param 4=&FloatFromCurve(FloatCurve=DataInterface() {propertyValues {Curve {keys [{arriveTangent=-0.136, leaveTangent=-0.136, time=0, value=1}]}}}, Scale Curve=0.3)
    Index 2 Param 1=&FloatFromCurve(FloatCurve=DataInterface() {propertyValues {Curve {keys [
         {arriveTangent=1.2401, interpMode=RCIM_Cubic, leaveTangent=0.0353, tangentMode=RCTM_User, time=0, value=0}
         {arriveTangent=0.0618, interpMode=RCIM_Cubic, leaveTangent=0.0618, tangentMode=RCTM_User, time=1, value=1}
        ]}}})
    Index 2 Param 2=&FloatFromCurve(
     FloatCurve=DataInterface() {propertyValues {Curve {keys [
         {arriveTangent=1.9056, time=0, value=0.2}
         {arriveTangent=0.0259, interpMode=RCIM_Cubic, leaveTangent=0.0259, tangentMode=RCTM_User, time=1, value=1.4}
        ]}}}
     Scale Curve=&RandomRangeFloat(Minimum=0.5, Maximum=1)
    )
    Index 2 Param 3=&FloatFromCurve(FloatCurve=DataInterface() {propertyValues {Curve {keys [
         {interpMode=RCIM_Cubic, leaveTangent=2.3064, tangentMode=RCTM_Break, time=0, value=1}
         {interpMode=RCIM_Cubic, leaveTangent=11.1233, tangentMode=RCTM_Break, time=0.5, value=2}
        ]}}})
    Index 2 Param 4=&FloatFromCurve(FloatCurve=DataInterface() {propertyValues {Curve {keys [
         {arriveTangent=-1.7751, interpMode=RCIM_Cubic, leaveTangent=-1.7751, tangentMode=RCTM_User, time=0, value=1}
         {interpMode=RCIM_Cubic, leaveTangent=-1.2956, tangentMode=RCTM_Break, time=1, value=0.2}
        ]}}})
    Module.Param0WriteEnabled=true
    Module.Param0WriteEnabled1=true
    Module.Param0WriteEnabled2=true
    Module.Param1WriteEnabled=false
    Module.Param1WriteEnabled1=true
    Module.Param1WriteEnabled2=false
    Module.Param2WriteEnabled=true
    Module.Param2WriteEnabled1=true
    Module.Param2WriteEnabled2=true
    Module.Param3WriteEnabled=true
    Module.Param3WriteEnabled1=true
    Module.Param3WriteEnabled2=true
    Write Parameter Index 1=true
    Write Parameter Index 2=true
   )
   &SolveForcesAndVelocity(Clamp Velocity=false, Limit Acceleration=false)
   &SpriteRotationRate(Rotation Rate=&Multiply_Float(
     A=$Particles.InitSpriteRotRate
     B=&FloatFromCurve(FloatCurve=DataInterface() {propertyValues {Curve {keys [<at -0.0046 1.0113 RCIM_Cubic>,<at 1.0011 0.09 RCIM_Cubic>]}}})
    ))
   &CameraOffset(Camera Offset Amount=20)
  ]
  rendererClasses [&NiagaraSpriteRendererProperties]
  renderers [
   {rendererClass=&NiagaraSpriteRendererProperties}
   {rendererClass=&NiagaraSpriteRendererProperties, rendererIndex=1}
  ]
 }
}
summary {
 systemName=NS_Powerup_MiniGun_Overheat
 emitters [
  {emitterName=Glow, rendererClasses [&NiagaraSpriteRendererProperties]}
  {emitterName=ShapesSmall, rendererClasses [&NiagaraMeshRendererProperties]}
  {emitterName=Corona, rendererClasses [&NiagaraSpriteRendererProperties]}
  {emitterName=ShapesBurst, rendererClasses [&NiagaraMeshRendererProperties]}
  {emitterName=Smoke, rendererClasses [&NiagaraSpriteRendererProperties]}
  {emitterName=ElectricityBoneFlowOverheat, rendererClasses [&NiagaraRibbonRendererProperties]}
  {emitterName=ElectricityBoneFlow2, rendererClasses [&NiagaraRibbonRendererProperties]}
  {emitterName=ElectricityBoneFlow3, rendererClasses [&NiagaraRibbonRendererProperties]}
  {emitterName=GlowContactPoints, rendererClasses [&NiagaraSpriteRendererProperties]}
  {bEnabled=false, emitterName=ElectricityBoneFlowOverheat001, rendererClasses [&NiagaraRibbonRendererProperties]}
  {bEnabled=false, emitterName=GlowContactPoints001, rendererClasses [&NiagaraSpriteRendererProperties]}
 ]
 userVariables [
  User.Overheat {type=NiagaraFloat, defaultValue=0, description=""}
  User.SkeletalMesh {
   type=NiagaraDataInterfaceSkeletalMesh
   defaultValue {
    struct=NiagaraExt_VariableValue_DataInterface
    value {dataInterface=&self:NiagaraDataInterfaceSkeletalMesh_0}
   }
  }
  User.SkeletalMesh2 {
   type=NiagaraDataInterfaceSkeletalMesh
   defaultValue {
    struct=NiagaraExt_VariableValue_DataInterface
    value {dataInterface=&self:NiagaraDataInterfaceSkeletalMesh_1}
   }
  }
  User.SkeletalMesh3 {
   type=NiagaraDataInterfaceSkeletalMesh
   defaultValue {
    struct=NiagaraExt_VariableValue_DataInterface
    value {dataInterface=&self:NiagaraDataInterfaceSkeletalMesh_3}
   }
  }
  User.SkeletalMesh_Overheat {
   type=NiagaraDataInterfaceSkeletalMesh
   defaultValue {
    struct=NiagaraExt_VariableValue_DataInterface
    value {dataInterface=&self:NiagaraDataInterfaceSkeletalMesh_2}
   }
  }
 ]
}
